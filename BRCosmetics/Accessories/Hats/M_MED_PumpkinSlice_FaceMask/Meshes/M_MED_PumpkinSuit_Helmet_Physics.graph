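# Physics asset for the pumpkin suit helmet hat cosmetic, covering the head, a bounce bone and a two-bone stem.
# Capsule bodies cover all four bones, with zero-length capsules acting as spheres on the head and bounce bone.
# Constraints chain head, bounce, lower stem and upper stem, locking swing and twist between head and bounce.
# The two stem joints share the same swing and twist angle settings, with unequal swing axes.
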PhysicsAsset /BRCosmetics/Accessories/Hats/M_MED_PumpkinSlice_FaceMask/Meshes/M_MED_PumpkinSuit_Helmet_Physics
bodies {
 dyn_bounce [{shapeType=Capsule, length=0, radius=10}]
 dyn_stem_01 [{shapeType=Capsule, length=3.3306, radius=2.3456, center=<1.9023 0 0>, rotation=<rot 90 0 0>}]
 dyn_stem_02 [{shapeType=Capsule, length=3.6046, radius=1.3805, center=<1.5481 0.0001 0>, rotation=<rot 90 -0.0201 -0.0204>}]
 head [{shapeType=Capsule, length=0, radius=2}]
}
constraints [
 {
  bone1Name=dyn_bounce
  bone2Name=head
  swing1LimitDegrees=10
  swing1Motion=Locked
  swing2LimitDegrees=10
  swing2Motion=Locked
  twistLimitDegrees=0
  twistMotion=Locked
 }
 {bone1Name=dyn_stem_02, bone2Name=dyn_stem_01, swing1LimitDegrees=45, swing2LimitDegrees=20, twistLimitDegrees=20}
 {bone1Name=dyn_stem_01, bone2Name=dyn_bounce, swing1LimitDegrees=45, swing2LimitDegrees=20, twistLimitDegrees=20}
]
